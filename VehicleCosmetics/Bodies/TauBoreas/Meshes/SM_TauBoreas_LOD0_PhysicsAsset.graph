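# Physics asset for the TauBoreas vehicle body cosmetic, named after its LOD0 mesh.
# A single body on the frame bone approximates its collision shape with one large, rotated capsule offset from the bone origin.
# It defines no constraints, so it has no jointed or per-part physics.
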PhysicsAsset /VehicleCosmetics/Bodies/TauBoreas/Meshes/SM_TauBoreas_LOD0_PhysicsAsset
bodies {frame [{shapeType=Capsule, length=180.0787, radius=106.7944, center=<-4.5232 -55.4185 -1.1937>, rotation=<rot 0 88.8046 -73.2141>}]}
constraints []
